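; Blueprint function library of mud gameplay helpers that query and clear mud-related gameplay effects on an actor.
; IsOnMud and IsCoveredInMud check granted gameplay tags, and RemoveMudCover removes active effects granting the covered-in-mud tag.
; TryRemoveSharedMudEffects strips effects tagged as shared mud only when the actor is neither on mud nor covered, returning true if it had a valid ability system component to act on.
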(fn CanRemoveSharedMudEffects
  (params (Actor :type Actor%20Object%20Reference))
  ;[1 1 1] Shared mud effects can only be removed if the actor is not on mud and not covered in mud
  (return (and (not (CallFunction|IsCoveredinMud Actor (Variables|Get__WorldContext))) (not (CallFunction|IsonMud Actor (Variables|Get__WorldContext))))))

(fn IsCoveredInMud
  (params (Actor :type Actor%20Object%20Reference))
  (bind _returnvalue (GameplayTags|HasMatchingGameplayTag(Message) Actor "(TagName=\"Granted.Mud.CoveredInMud\")"))
  (return _returnvalue))

(fn IsOnMud
  (params (Actor :type Actor%20Object%20Reference))
  (bind _returnvalue (GameplayTags|HasMatchingGameplayTag(Message) Actor "(TagName=\"Granted.Mud.OnMud\")"))
  (return _returnvalue))

(fn RemoveMudCover
  (params (Actor :type Actor%20Object%20Reference))
  (bind _returnvalue (Ability|GetAbilitySystemComponent Actor))
  (Utilities|IsValid _returnvalue
    (:Is%20Valid
      (GameplayEffects|RemoveActiveEffectswithGrantedTags _returnvalue "(GameplayTags=((TagName=\"Granted.Mud.CoveredInMud\")))"))))

(fn TryRemoveSharedMudEffects
  (params (Actor :type Actor%20Object%20Reference))
  (bind _returnvalue (Ability|GetFortAbilitySystemComponent Actor))
  (if (CallFunction|CanRemoveSharedMudEffects Actor (Variables|Get__WorldContext))
    (Utilities|IsValid _returnvalue
      (:Is%20Valid
        (GameplayEffects|RemoveActiveEffectswithTags _returnvalue "(GameplayTags=((TagName=\"GameplayEffect.Mud.Shared\")))")
        (return true))
      (:Is%20Not%20Valid
        _))
    (else
      _))
  (return false))
